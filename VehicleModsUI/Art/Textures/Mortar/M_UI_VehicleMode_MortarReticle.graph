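# Translucent UI material for the vehicle-mod mortar reticle, combining a box frame, drop lines and side lines.
# The frame is an SDF box outline multiplied by a mask, with a separate sine-of-UV mask providing adjustable gaps.
# Drop lines sample a scaled, offset distance-field launcher crosshair texture; side lines are drawn procedurally with UV symmetry.
# Scalar parameters control shape and weight each element's opacity before summing.
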Material /VehicleModsUI/Art/Textures/Mortar/M_UI_VehicleMode_MortarReticle
graph {
 properties {blendMode=BLEND_Translucent, materialDomain=MD_UI, pixelDepthOffsetMode=PDOM_Legacy, usedShadingModels=Default Lit}
 nodes [
  n1 = NamedRerouteUsage()
  n2 = ScalarParameter() {defaultValue=1, parameterName=Opacity}
  n3 = NamedRerouteUsage()
  n4 = ScalarParameter() {defaultValue=1, parameterName=OpacityDropLines}
  Add(
   A=Add(A=Multiply(A=@n1, B=@n2), B=Multiply(A=@n3, B=@n4))
   B=Multiply(A=NamedRerouteUsage(), B=ScalarParameter() {defaultValue=1, parameterName=OpacitySideLines})
  )
  Constant3Vector() {constant=<rgba 1 1 1 0>}
  n5 = ScalarParameter() {parameterName=Thickness}
  n6 = TextureCoordinate()
  n7 = ComponentMask(None=@n6) {r=true}
  n8 = Add(A=ComponentMask(None=@n6) {g=true}, B=ScalarParameter() {parameterName=BoxFramePosition})
  n9 = SmoothStep(
   Min=Add(A=Constant(), B=@n5)
   Max=Add(A=@n5, B=Constant() {r=0.01})
   Value=MaterialFunctionCall(UVs=AppendVector(A=@n7, B=@n8), Size=ScalarParameter() {defaultValue=0.3, parameterName=RectangleSize}, OutlineThickness=Constant() {r=0.01}) {materialFunction=&MF_UI_Box}.SDFOutline
  ) {constMax=0.01}
  NamedRerouteDeclaration(None=Multiply(A=OneMinus(None=@n9), B=NamedRerouteUsage())) {name=BoxFrame, nodeColor=<rgba 0 0 0 1>}
  n10 = Constant() {r=0.1}
  n11 = TextureCoordinate()
  n12 = Add(A=Sine(None=Add(A=@n11) {constB=-0.25}), B=ScalarParameter() {defaultValue=-0.85, parameterName=GapSize}) {constB=-0.7}
  n13 = Add(
   A=SmoothStep(Max=@n10, Value=ComponentMask(None=@n12) {r=true}) {constMax=0.01}
   B=SmoothStep(Max=@n10, Value=ComponentMask(None=@n12) {g=true}) {constMax=0.01}
  )
  NamedRerouteDeclaration(None=OneMinus(None=Ceil(None=@n13))) {name=Mask, nodeColor=<rgba 0 0 0 1>}
  n14 = TextureCoordinate()
  n15 = ComponentMask(None=@n14) {g=true}
  n16 = ScalarParameter() {defaultValue=-0.436, parameterName=BottomPartPosition}
  n17 = MaterialFunctionCall(
   UVs=AppendVector(A=ComponentMask(None=@n14) {r=true}, B=Add(A=@n15, B=@n16))
   Texture Scale=ScalarParameter() {defaultValue=0.5, parameterName=BottomScale}
  ) {materialFunction=&ScaleUVsByCenter}
  n18 = Constant() {r=0.196}
  n19 = ScalarParameter() {parameterName=DropLinesThickness}
  n20 = Constant() {r=0.212}
  NamedRerouteDeclaration(None=MaterialFunctionCall!(
    RGBA Channels=TextureSample(UVs=@n17.UVs) {samplerSource=SSM_Clamp_WorldGroupSettings, texture=&/T_UI_HUD_Xhair_Launcher_HLines}.R
    Glow Min=Subtract(A=@n18, B=@n19)
    Glow Max=Subtract(A=@n20, B=@n19)
    Outline Thickness
   ) {materialFunction=/FortUILibrary/MaterialFunctions/Basics/MF_UI_ChannelPackedDistanceField}.Fill) {name=BottomPart, nodeColor=<rgba 0 0 0 1>}
  n21 = MaterialFunctionCall!(
   UV
   Offset=AppendVector(A=ScalarParameter() {defaultValue=-0.072, parameterName=OffsetX}, B=Constant())
   Rotation
   Scale=Constant2Vector() {g=1, r=1}
   MirrorY
  ) {materialFunction=/FortUILibrary/MaterialFunctions/Basics/MF_UI_Symmetry_2x1_DEPRECATED}
  n22 = Constant2Vector() {g=0.5}
  n23 = Constant2Vector() {g=0.5, r=0.4}
  NamedRerouteDeclaration(None=Multiply(A=MaterialFunctionCall!(Coordinates=@n21.Result, P0=@n22, P1=@n23, Line Width, Edge Falloff, Square Corners?) {materialFunction=/Engine/Functions/Engine_MaterialFunctions01/Coordinates/DrawLine-2D}.Result)) {name=Line, nodeColor=<rgba 0 0 0 1>}
 ]
}
