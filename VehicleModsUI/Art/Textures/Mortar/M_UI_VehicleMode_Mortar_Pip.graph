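# Translucent UI material for a pip in the mortar vehicle-mode display, drawn as a narrow rectangle that fills according to a Percent parameter.
# A box function supplies the rectangle mask, while a rotated linear gradient offset by Percent passes through a tight SmoothStep and inversion to form a hard-edged progress mask.
# Named reroutes share these masks with the blend math, where FillOpacity and ProgressOpacity scale terms combined by Min and lerps.
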Material /VehicleModsUI/Art/Textures/Mortar/M_UI_VehicleMode_Mortar_Pip
graph {
 properties {
  bAutomaticallySetUsageInEditor=false
  blendMode=BLEND_Translucent
  materialDomain=MD_UI
  pixelDepthOffsetMode=PDOM_Legacy
  usedShadingModels=Default Lit
 }
 nodes [
  n1 = Min(
   A=Multiply(A=NamedRerouteUsage(), B=ScalarParameter() {defaultValue=0.5, parameterName=ProgressOpacity})
   B=Multiply(A=NamedRerouteUsage(), B=ScalarParameter() {defaultValue=1, parameterName=FillOpacity})
  )
  LinearInterpolate(A=Multiply(A=NamedRerouteUsage()) {constB=0.5}, B=@n1, Alpha=@n1)
  LinearInterpolate(A=Constant3Vector() {constant=<rgba 1 1 1 1>}, B=Constant3Vector(), Alpha=OneMinus(None=NamedRerouteUsage()))
  n2 = MaterialFunctionCall(RotationAngle=Constant() {r=0.25}, GradientOffset=ScalarParameter() {parameterName=Percent}) {materialFunction=&MF_UI_LinearGradient}
  NamedRerouteDeclaration(None=OneMinus(None=SmoothStep(Value=@n2.Result) {constMax=0.01})) {name=Progress, nodeColor=<rgba 0 0 0 1>}
  NamedRerouteDeclaration(None=MaterialFunctionCall(Size=Constant2Vector() {g=1, r=0.4}) {materialFunction=&MF_UI_Box}.Fill) {name=RectangleFill, nodeColor=<rgba 0 0 0 1>}
 ]
}
